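# Material function for the Polished Jade Mind outfit that adds animated glow to incoming material attributes, overriding base color and emissive.
# An "is helmet?" scalar switches between body glow, built from masked Fresnel and panning noise, and helmet fire.
# The fire uses noise-distorted mask UVs, with a UV-based directional mask available for pan direction.
# A night-only gate combines Scalar_Time_of_Day with a manual Glow ON OFF input.
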MaterialFunction /BRCosmetics/Characters/Player/Female/Medium/Bodies/F_MED_PolishedJade_Mind/Materials/MF_PolishedJade_Mind_E
graph {nodes [
  n1 = GetMaterialAttributes(None=FunctionInput() {inputType=FunctionInput_MaterialAttributes}) {attributeGetTypes [#1, #2]}
  n2 = NamedRerouteUsage()
  n3 = Multiply(A=TextureSampleParameter2D() {group=PolishedJade_mind, parameterName=Helmet, texture=&T_BlacksRGB}.R) {constB=10}
  n4 = LinearInterpolate(
   A=NamedRerouteUsage()
   B=VectorParameter() {group=PolishedJade_mind, parameterName=BaseColorE, defaultValue=<rgba 0 0.2573 0.35 1>}.RGB
   Alpha=LinearInterpolate(
    A=TextureSampleParameter2D() {group=PolishedJade_mind, parameterName=BodyMask, texture=&T_BlacksRGB}.G
    B=Saturate(None=@n3)
    Alpha=ScalarParameter() {group=PolishedJade_mind, parameterName=is helmet?}
   )
  )
  n5 = NamedRerouteUsage()
  FunctionOutput(None=SetMaterialAttributes!(MaterialAttributes=@n1.MaterialAttributes, EmissiveColor=NamedRerouteUsage(), BaseColor=LinearInterpolate(A=@n2, B=@n4, Alpha=@n5)) {attributeSetTypes [#1, #2]})
  n6 = TextureSampleParameter2D() {group=PolishedJade_mind, parameterName=Helmet, texture=&T_BlacksRGB}
  n7 = TextureCoordinate() {uTiling=0.8}
  n8 = Constant2Vector() {g=0.1, r=0.2}
  n9 = Constant2Vector() {g=-0.1, r=-0.2}
  n10 = NamedRerouteUsage()
  n11 = TextureSample(UVs=Panner(Coordinate=Multiply(A=@n7) {constB=5}, Speed=LinearInterpolate(A=@n8, B=@n9, Alpha=@n10)) {speedX=-0.2, speedY=-0.1}) {samplerSource=SSM_Wrap_WorldGroupSettings, samplerType=SAMPLERTYPE_Grayscale, texture=&LowResBlurredNoise}
  n12 = TextureCoordinate() {uTiling=0.8}
  n13 = Constant2Vector() {g=0.1, r=0.4}
  n14 = Constant2Vector() {g=-0.1, r=-0.5}
  n15 = NamedRerouteUsage()
  n16 = TextureSample(UVs=Panner(Coordinate=Multiply(A=@n12) {constB=10}, Speed=LinearInterpolate(A=@n13, B=@n14, Alpha=@n15)) {speedX=-0.4, speedY=-0.1}) {samplerSource=SSM_Wrap_WorldGroupSettings, samplerType=SAMPLERTYPE_Grayscale, texture=&LowResBlurredNoise}
  n17 = Max(A=@n11.R, B=@n16.R)
  n18 = Multiply(A=@n17) {constB=-0.025}
  n19 = TextureCoordinate()
  n20 = Multiply(
   A=Power(Base=LinearInterpolate(Alpha=@n17) {constA=0.4})
   B=TextureSampleParameter2D(UVs=Add(A=@n18, B=@n19)) {group=PolishedJade_mind, parameterName=Helmet, texture=&T_BlacksRGB}.G
  )
  n21 = TextureSampleParameter2D() {group=PolishedJade_mind, parameterName=Helmet, texture=&T_BlacksRGB}
  n22 = Add(
   A=Multiply(A=Power(Base=@n21.G) {constExponent=5}) {constB=0.4}
   B=Power(Base=TextureSampleParameter2D(UVs=@n18) {group=PolishedJade_mind, parameterName=Helmet, texture=&T_BlacksRGB}.G)
  )
  n23 = Multiply(
   A=Max(A=Multiply(A=@n6.B) {constB=0.5}, B=Max(A=@n20, B=@n22))
   B=VectorParameter() {group=PolishedJade_mind, parameterName=Helemt Fire Color, defaultValue=<rgba 0.375 1.2549 1.5 1>}.RGB
  )
  n24 = TextureCoordinate()
  n25 = TextureSample(UVs=Panner(Coordinate=Multiply(A=@n24) {constB=2}) {speedX=-0.15}) {samplerSource=SSM_Wrap_WorldGroupSettings, samplerType=SAMPLERTYPE_Grayscale, texture=&LowResBlurredNoise}
  n26 = TextureSampleParameter2D() {group=PolishedJade_mind, parameterName=Helmet, texture=&T_BlacksRGB}
  n27 = Power(Base=@n26.R)
  n28 = Multiply(
   A=Multiply(A=LinearInterpolate(Alpha=@n25.R) {constA=0.4}, B=Saturate(None=@n27))
   B=LinearInterpolate(
    A=VectorParameter() {group=PolishedJade_mind, parameterName=HelmetColorB, defaultValue=<rgba 0 1.5436 3 1>}.RGB
    B=VectorParameter() {group=PolishedJade_mind, parameterName=HelmetColorA, defaultValue=<rgba 0 3.9109 5 1>}.RGB
    Alpha=@n26.R
   )
  )
  n29 = Multiply(
   A=LinearInterpolate(A=NamedRerouteUsage(), B=Add(A=@n23, B=@n28), Alpha=ScalarParameter() {group=PolishedJade_mind, parameterName=is helmet?})
   B=NamedRerouteUsage()
  )
  MaterialFunctionCall(Emissive=@n29) {materialFunction=&MF_FX_EmissiveAdjustments}
  n30 = Power(Base=LinearInterpolate(Alpha=Fresnel() {exponent=2}))
  n31 = TextureCoordinate()
  n32 = TextureSample(UVs=Panner(Coordinate=Multiply(A=@n31) {constB=3.5}) {speedX=-0.15}) {samplerSource=SSM_Wrap_WorldGroupSettings, samplerType=SAMPLERTYPE_Grayscale, texture=&LowResBlurredNoise}
  n33 = TextureSampleParameter2D() {group=PolishedJade_mind, parameterName=BodyMask, texture=&T_BlacksRGB}
  n34 = Multiply(A=Multiply(A=Multiply(A=@n30, B=@n32.R), B=@n33.R), B=Constant() {r=2}) {constB=10}
  n35 = Multiply(A=LinearInterpolate(Alpha=@n32.R) {constA=0.2}, B=@n33.R)
  n36 = VectorParameter() {group=PolishedJade_mind, parameterName=Body_colorB, defaultValue=<rgba 0.12 4.4425 6 1>}
  n37 = VectorParameter() {group=PolishedJade_mind, parameterName=Body_colorA, defaultValue=<rgba 0.4871 3 2.5497 1>}
  NamedRerouteDeclaration(None=Multiply(A=Add(A=@n34, B=@n35), B=LinearInterpolate(A=@n36.RGB, B=@n37.RGB, Alpha=@n33.R))) {name=Body, nodeColor=<rgba 0 0.1094 1 1>}
  n38 = Constant() {r=0.462}
  n39 = TextureCoordinate()
  NamedRerouteDeclaration(None=SmoothStep(Min=@n38, Max=@n38, Value=ComponentMask(None=@n39) {r=true}) {constMax=0.5, constMin=0.5}) {name=Directional mask, nodeColor=<rgba 1 0.9141 0 1>}
  n40 = Constant()
  n41 = Constant() {r=1}
  NamedRerouteDeclaration(None=Max(
    A=MaterialFunctionCall(Value_Morning=@n40, Value_Day=@n40, Value_Evening=@n40, Value_Night=@n41) {materialFunction=&Scalar_Time_of_Day}.Result
    B=FunctionInput() {bUsePreviewValueAsDefault=true, inputName=Glow ON OFF, inputType=FunctionInput_Scalar}
   )) {name=Glow only at night, nodeColor=<rgba 0.5938 0 1 1>}
  NamedRerouteDeclaration(None=@n1.EmissiveColor) {name=In_EmissiveColor, nodeColor=<rgba 0 1 0.2734 1>}
  NamedRerouteDeclaration(None=@n1.BaseColor) {name=InBaseColor, nodeColor=<rgba 1 0 0.0469 1>}
  n42 = NamedRerouteUsage()
  n43 = Multiply(A=@n42) {constB=1.5}
  n44 = NamedRerouteUsage()
  NamedRerouteDeclaration(None=Add(A=@n29, B=LinearInterpolate(A=@n42, B=@n43, Alpha=@n44))) {name=Out EmissiveColor, nodeColor=<rgba 0 0.3672 1 1>}
 ]}
