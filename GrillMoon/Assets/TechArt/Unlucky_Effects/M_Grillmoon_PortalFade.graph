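# Material function that clips and tints objects or particles crossing the Grill Moon portal, with a glowing edge.
# Opacity uses either a world-Y cutoff at the portal or, with UsePortalSilhouette, a ray-disk intersection with dithered fresnel.
# A box mask around the portal blends glow into emissive and base color.
# Silhouette mode also adjusts surface attributes by depth past the portal and applies camera-offset WPO toward the disk plane.
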MaterialFunction /GrillMoon/Assets/TechArt/Unlucky_Effects/M_Grillmoon_PortalFade
graph {nodes [
  BumpOffset(Coordinate=TextureCoordinate(), Height=ScalarParameter() {defaultValue=-12.5039}) {referencePlane=0}
  Constant3Vector() {desc=glow color og, constant=<rgba 0.276 0.0316 0.1012 1>}
  Constant3Vector() {desc=glow color highlight, constant=<rgba 1 0.0573 0.3258 1>}
  Constant3Vector() {desc=base color og, constant=<rgba 0.1771 0.0203 0.0649 1>}
  Constant() {r=42904}
  Constant() {r=42919.8984}
  Constant() {r=42914.8984}
  n1 = NamedRerouteUsage()
  n2 = Saturate(None=LinearInterpolate(Alpha=@n1) {constA=1, constB=-1})
  n3 = Saturate(None=LinearInterpolate(Alpha=@n1) {constA=2, constB=0})
  n4 = VertexNormalWS()
  n5 = CameraVectorWS()
  n6 = OneMinus(None=Saturate(None=DotProduct(A=@n4, B=@n5)))
  n7 = OneMinus(None=Saturate(None=InverseLinearInterpolate(A=@n2, B=@n3, Value=@n6)))
  n8 = Normalize(VectorInput=ComponentMask(None=VectorParameter() {group=Silhouette, parameterName=Disk Normal, defaultValue=<rgba 1 0 0 1>}.RGB) {b=true, g=true, r=true})
  n9 = VectorParameter() {group=Silhouette, parameterName=Disk Center + Radius, defaultValue=<rgba -200 0 0 200>}
  n10 = ComponentMask(None=@n9.RGB) {b=true, g=true, r=true}
  n11 = Normalize(VectorInput=Subtract(A=WorldPosition() {worldPositionShaderOffset=WPT_ExcludeAllShaderOffsets}.XYZ, B=@n10))
  n12 = Ceil(None=Saturate(None=DotProduct(A=@n8, B=@n11)))
  n13 = Constant()
  n14 = Constant() {r=1}
  n15 = If(A=@n12, "A > B"=@n13, "A == B"=@n14, "A < B"=@n14)
  n16 = Subtract(A=CameraPositionWS() {desc=Ray Origin}, B=@n10)
  n17 = Subtract(A=WorldPosition() {worldPositionShaderOffset=WPT_ExcludeAllShaderOffsets}.XYZ, B=CameraPositionWS()) {desc=Direction Vector}
  n18 = DotProduct(A=@n17, B=@n8)
  n19 = Divide(A=Multiply(A=DotProduct(A=@n8, B=@n16)) {constB=-1}, B=@n18)
  n20 = Add(A=@n16, B=Multiply(A=@n17, B=@n19))
  n21 = Constant() {r=-1}
  n22 = If(A=DotProduct(A=@n20, B=@n20), B=Multiply(A=@n9.A, B=@n9.A), "A > B"=@n21, "A == B"=@n21, "A < B"=@n19)
  n23 = Saturate(None=@n18)
  n24 = Saturate(None=Ceil(None=Multiply(A=@n22, B=@n23)))
  n25 = Divide(A=LocalPosition() {localOrigin=InstancePreSkinning}.XYZ, B=PreSkinnedLocalBounds().Extents)
  n26 = StaticSwitchParameter(
   True=ParticlePositionWS()
   False=Subtract(A=WorldPosition() {worldPositionShaderOffset=WPT_ExcludeAllShaderOffsets}.XYZ, B=VertexInterpolator(VS=@n25).PS)
  ) {group=Grill_Moon, parameterName=Particle_Position}
  n27 = ComponentMask(None=@n26) {g=true}
  n28 = VectorParameter() {
   desc=was 42919 for the Y
   group=Grill_Moon
   parameterName=Portal_Location
   channelNames {a="", b=Z, g=Y, r=X}
   defaultValue=<rgba -57597.9531 42919.8984 10468.1201 1>
  }
  n29 = StaticSwitchParameter(
   True=Multiply(
    A=MaterialFunctionCall(Alpha Threshold=@n7) {materialFunction=&DitherTemporalAA}.Result
    B=If(A=@n15, "A > B"=@n15, "A == B"=@n24, "A < B"=@n24)
   )
   False=If(A=@n27, B=@n28.Y, "A > B"=@n13, "A == B"=@n13, "A < B"=@n14)
  ) {group=Silhouette, parameterName=UsePortalSilhouette}
  FunctionOutput(None=@n29) {description=Use to control visibility of object, outputName=Opacity, sortPriority=1}
  n30 = LinearInterpolate(
   A=VectorParameter() {group=Grill_Moon, parameterName=Glow_Color, defaultValue=<rgba 0.276 0.0316 0.1012 1>}.RGB
   B=VectorParameter() {group=Grill_Moon, parameterName=Glow_Color_Highlight, defaultValue=<rgba 1 0.1198 0.2666 1>}.RGB
   Alpha=NamedRerouteUsage()
  )
  n31 = MaterialFunctionCall!(
   A=@n26
   B=@n28
   Bounds=VectorParameter() {group=Grill_Moon, parameterName=Edge_Fade_Distance, defaultValue=<rgba 1000 50 1000 1>}.RGB
   Edge Falloff=LinearInterpolate(
    A=ScalarParameter() {defaultValue=30, group=Grill_Moon, parameterName=Edge_Falloff}
    B=ScalarParameter() {defaultValue=40, parameterName=edge_falloff_low}
    Alpha=NamedRerouteUsage()
   )
  ) {materialFunction=/Engine/Functions/Engine_MaterialFunctions01/Gradient/BoxMask-3D}
  FunctionOutput(None=LinearInterpolate(
    A=FunctionInput() {bUsePreviewValueAsDefault=true, description=Only needed if just tweaking emissive value manually, inputName=Emissive}
    B=@n30
    Alpha=@n31.Result
   )) {outputName=Edge_Glow, sortPriority=2}
  n32 = GetMaterialAttributes(None=FunctionInput() {bUsePreviewValueAsDefault=true, inputName=Mat Attributes, inputType=FunctionInput_MaterialAttributes}) {attributeGetTypes [
    #1
    #2
    #3
    #4
   ]}
  n33 = SetMaterialAttributes!(
   MaterialAttributes=@n32.MaterialAttributes
   OpacityMask=@n29
   EmissiveColor=NamedRerouteUsage()
   BaseColor=NamedRerouteUsage()
   WorldPositionOffset=NamedRerouteUsage()
   Roughness=NamedRerouteUsage()
   Specular=NamedRerouteUsage()
  ) {attributeSetTypes [
    #5
    #1
    #2
    #6
    #3
    #4
   ]}
  FunctionOutput(None=@n33) {outputName=Mat Output}
  InverseLinearInterpolate(Value=@n1) {constB=0.75}
  n34 = VectorParameter() {parameterName=Base_Color, defaultValue=<rgba 0.1771 0.0203 0.0649 1>}
  LinearInterpolate(A=@n34.RGB, Alpha=NamedRerouteUsage()) {constB=0}
  NamedRerouteDeclaration(None=@n32.EmissiveColor) {name=Mat_Emissive, nodeColor=<rgba 0.8516 0 1 1>}
  n35 = LinearInterpolate(A=NamedRerouteUsage(), B=@n30, Alpha=@n31.Result)
  n36 = NamedRerouteUsage()
  NamedRerouteDeclaration(None=StaticSwitchParameter(True=LinearInterpolate(A=@n35, Alpha=@n36) {constB=0}, False=@n35) {group=Silhouette, parameterName=UsePortalSilhouette}) {name=output emissive, nodeColor=<rgba 1 0.2109 0 1>}
  n37 = MaterialFunctionCall(Float2=NamedRerouteUsage()) {materialFunction=&BreakOutFloat2Components}
  n38 = Multiply(A=@n37.R, B=ScalarParameter() {defaultValue=1, parameterName=UV_Tilt})
  NamedRerouteDeclaration(None=MaterialFunctionCall(X=@n37.R, Y=Add(A=@n37.G, B=@n38)) {materialFunction=&MakeFloat2}.Result) {name=TiltedUV, nodeColor=<rgba 0 0.3848 1 1>}
  NamedRerouteDeclaration(None=@n32.Roughness) {name=Mat_Roughness, nodeColor=<rgba 1 0 0.5391 1>}
  NamedRerouteDeclaration(None=@n32.Specular) {name=Mat_Specular, nodeColor=<rgba 0 0.8594 1 1>}
  n39 = NamedRerouteUsage()
  NamedRerouteDeclaration(None=StaticSwitchParameter(True=LinearInterpolate(A=@n39, Alpha=@n36), False=@n39) {group=Silhouette, parameterName=UsePortalSilhouette}) {name=output_roughness, nodeColor=<rgba 1 0.8203 0 1>}
  n40 = NamedRerouteUsage()
  NamedRerouteDeclaration(None=StaticSwitchParameter(True=LinearInterpolate(A=@n40, Alpha=@n36) {constB=0}, False=@n40) {group=Silhouette, parameterName=UsePortalSilhouette}) {name=output_specular, nodeColor=<rgba 0.5 0 1 1>}
  NamedRerouteDeclaration(None=@n32.BaseColor) {name=Mat_BaseColor, nodeColor=<rgba 0 1 0.2734 1>}
  n41 = LinearInterpolate(A=NamedRerouteUsage(), B=@n34.RGB, Alpha=@n31.Result)
  NamedRerouteDeclaration(None=StaticSwitchParameter(True=LinearInterpolate(A=@n41, Alpha=@n36) {constB=0}, False=@n41) {group=Silhouette, parameterName=UsePortalSilhouette}) {name=output_base_Color, nodeColor=<rgba 1 0 0.0469 1>}
  n42 = Add(A=NamedRerouteUsage()) {constB=-0.5}
  n43 = VectorParameter() {parameterName=Main_UV, defaultValue=<rgba 1 0.5 -0.1307 0.248>}
  NamedRerouteDeclaration(None=MaterialFunctionCall!(
    ClampTexture
    UV=MaterialFunctionCall(Vector or UVs=@n42) {materialFunction=&VectorToRadialValue}.Radial Coordinates
    Tex
    InRGB=@n43.RGB
    InA=@n43.RGB
    Time
    TimeMult
   ) {materialFunction=/Hermes/Props/ClayTile/ClayTile_DragonGate_B/VFX/MF_MOON_TexPanner_V2}.PannerUV) {name=MainUV, nodeColor=<rgba 0 0.4049 1 1>}
  n44 = ScalarParameter() {defaultValue=27, parameterName=size}
  NamedRerouteDeclaration(None=Multiply(A=WorldPosition().XY, B=@n44)) {name=BumpUV, nodeColor=<rgba 0.6406 0 1 1>}
  n45 = LocalPosition() {localOrigin=InstancePreSkinning}
  n46 = ComponentMask(None=@n45.XYZ) {r=true}
  n47 = Add(A=ComponentMask(None=@n45.XYZ) {g=true}, B=Multiply(A=Time()))
  n48 = MaterialFunctionCall(
   TextureObject=TextureObject() {samplerType=SAMPLERTYPE_LinearColor, texture=/Hermes/Props/ClayTile/ClayTile_DragonGate_B/VFX/Tile_Tech_02}
   TextureSize=@n44
   WorldPosition=VertexInterpolator(VS=MaterialFunctionCall(R=@n46, G=@n47, B=@n45.Z) {materialFunction=&AppendMany}.RGB).PS
   World Space Normal=VertexInterpolator(VS=PreSkinnedNormal()).PS
  ) {materialFunction=&WorldAlignedTexture}
  n49 = ScalarParameter() {defaultValue=2.5, parameterName=noise_Str}
  NamedRerouteDeclaration(None=Power(Base=Multiply(A=@n48.XYZ Texture, B=@n49), Exp=ScalarParameter() {defaultValue=2, parameterName=Noise_Pow})) {name=noise, nodeColor=<rgba 1 0.3047 0 1>}
  n50 = Subtract(A=CameraPositionWS(), B=WorldPosition() {worldPositionShaderOffset=WPT_ExcludeAllShaderOffsets}.XYZ)
  n51 = Constant3Vector() {constant=<rgba 0 0 1 0>}
  n52 = Subtract(
   A=CameraPositionWS()
   B=MaterialFunctionCall(Plane Normal Axis=@n8, Plane X Axis=Normalize(VectorInput=@n51), Plane Center=@n10, UV Scale=@n9.A) {materialFunction=&VirtualPlaneCoordinates}.World Position
  )
  n53 = Max(A=Subtract(A=Length(None=@n50) {is=class MaterialExpressionLength}, B=Length(None=@n52))) {constB=0}
  n54 = ScalarParameter() {defaultValue=0.95, group=Silhouette, parameterName=Camera Offset Strength}
  n55 = MaterialFunctionCall(
   Offset Amount=Add(A=Multiply(A=@n53, B=@n54), B=ScalarParameter() {group=Silhouette, parameterName=Additional Camera Offset})
   Clamp Padding=ScalarParameter() {defaultValue=100, group=Silhouette, parameterName=Camera Offset Padding}
  ) {materialFunction=/Engine/Functions/Engine_MaterialFunctions02/WorldPositionOffset/CameraOffset}
  NamedRerouteDeclaration(None=StaticSwitchParameter(True=Multiply(A=@n55.Clamped Camera Offset, B=@n12) {constA=1}, False=Constant()) {group=Silhouette, parameterName=UsePortalSilhouette}) {name=output wpo, nodeColor=<rgba 0.0547 1 0 1>}
  n56 = InverseLinearInterpolate(A=@n28.Y, B=Subtract(A=@n28.Y, B=ScalarParameter() {defaultValue=-150, group=Silhouette, parameterName=SilhouetteDistance}), Value=@n27)
  NamedRerouteDeclaration(None=InverseLinearInterpolate(B=ScalarParameter() {defaultValue=0.5, group=Silhouette, parameterName=Silhouette Depth}, Value=Saturate(None=@n56))) {name=Silhouette Lerp, nodeColor=<rgba 1 0 0.4219 1>}
  ObjectPositionWS()
  n57 = TextureSampleParameter2D!(UVs=NamedRerouteUsage(), Level=ScalarParameter() {defaultValue=1.5, parameterName=Param_6}, Apply View MipBias) {
   mipValueMode=TMVM_MipLevel
   parameterName=Param
   samplerType=SAMPLERTYPE_LinearColor
   texture=&Tile_Tech_02
  }
  SquareRoot(None=@n57.G)
  WorldPosition()
 ]}
