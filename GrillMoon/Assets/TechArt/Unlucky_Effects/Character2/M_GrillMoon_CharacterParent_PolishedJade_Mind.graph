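# Masked master character material for the Polished Jade Mind character in GrillMoon, usable on skeletal, Nanite and Niagara meshes.
# Shared CharacterParent PreFX output feeds the Mind and MindGloom effect functions, with the IsMindGloom static switch choosing between them.
# Two parameters multiply into a saturated glow control for both variants.
# The chosen result runs through shared PostFX, then GrillMoon PortalFade, using material attributes with expression-driven shading models.
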Material /GrillMoon/Assets/TechArt/Unlucky_Effects/Character2/M_GrillMoon_CharacterParent_PolishedJade_Mind
graph {
 properties {
  bUseMaterialAttributes=true
  bUsedWithClothing=true
  bUsedWithMeshDeformer=true
  bUsedWithMorphTargets=true
  bUsedWithNanite=true
  bUsedWithNiagaraMeshParticles=true
  bUsedWithSkeletalMesh=true
  blendMode=BLEND_Masked
  pixelDepthOffsetMode=PDOM_Legacy
  shadingModel=MSM_FromMaterialExpression
  usedShadingModels=Default Lit | Subsurface | Clear Coat | Subsurface Profile | Cloth
 }
 nodes [
  n1 = Saturate(None=Multiply(A=ScalarParameter() {parameterName=GlowPolishedJade}, B=ScalarParameter() {parameterName=GlowPolishedJade_OnOff}))
  n2 = MaterialFunctionCall() {materialFunction=&MF_CharacterParent_2023_PreFX}
  n3 = MaterialFunctionCall!(Glow ON OFF=@n1, In=@n2.Result) {materialFunction=/BRCosmetics/Characters/Player/Female/Medium/Bodies/F_MED_PolishedJade_Mind/Materials/MF_PolishedJade_MindGloom_E}
  n4 = MaterialFunctionCall!(Glow ON OFF=@n1, In=@n2.Result) {materialFunction=/BRCosmetics/Characters/Player/Female/Medium/Bodies/F_MED_PolishedJade_Mind/Materials/MF_PolishedJade_Mind_E}
  MaterialFunctionCall!(
   Mat Attributes=MaterialFunctionCall(
    Material=StaticSwitchParameter(True=@n3.Result, False=@n4.Result) {parameterName=IsMindGloom}
    M=@n2.M
    Pre Skinned Local Position=@n2.M
    Distance Blend=@n2.M
    Fresnel=@n2.M
   ) {materialFunction=&MF_CharacterParent_2023_PostFX}.Result
   Emissive
  ) {materialFunction=/GrillMoon/Assets/TechArt/Unlucky_Effects/M_Grillmoon_PortalFade}
 ]
}
