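# Parent master material for GrillMoon character eyes on skeletal meshes, using masked blend and a shading model set from the expression.
# The shared MF_eyes function builds the eye attributes, which pass through MF_TechArtOverrides and then into the GrillMoon PortalFade function.
# All eye and fade logic lives in those material functions, so this graph only chains them through material attributes.
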Material /GrillMoon/Assets/TechArt/Unlucky_Effects/Character2/M_GrillMoon_Eyes_Parent
graph {
 properties {
  bUseMaterialAttributes=true
  bUsedWithMeshDeformer=true
  bUsedWithSkeletalMesh=true
  blendMode=BLEND_Masked
  pixelDepthOffsetMode=PDOM_Legacy
  shadingModel=MSM_FromMaterialExpression
  usedShadingModels=Default Lit
 }
 nodes [MaterialFunctionCall!(
   Mat Attributes=MaterialFunctionCall(In=MaterialFunctionCall() {materialFunction=/Game/Characters/Shared/FNEye/Materials/MF_eyes}.Result) {materialFunction=&MF_TechArtOverrides}.Result
   Emissive
  ) {materialFunction=/GrillMoon/Assets/TechArt/Unlucky_Effects/M_Grillmoon_PortalFade}]
}
